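# Sound cue for the first-person firing sound of the drum shotgun.
# A quality-level root sends its first slot to a random pick of three close-fire waves and its other three slots to a shared random pick of two.
# Each random branch passes through a Modulator at default settings for pitch and volume variation.
# First-person weapon attenuation with overrides, a ranged-fire sound class and a gunfire concurrency set govern playback.
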SoundCue /Game/Sounds/Fort_Weapons/Shotgun/Drum/ShotgunDrum_Fire_1P_Cue
root=QualityLevel_0
nodes [
 {id=Modulator_0, children [Random_1]}
 {id=Modulator_7, children [Random_2]}
 {id=QualityLevel_0, children [Modulator_7,Modulator_0,Modulator_0,Modulator_0]}
 {id=Random_1, children [WavePlayer_3,WavePlayer_4], properties {weights [1,1]}}
 {id=Random_2, children [WavePlayer_7, WavePlayer_9, WavePlayer_10], properties {weights [1,1,1]}}
 {id=WavePlayer_10, properties {soundWaveAssetPtr=&/ShotgunDrum_Fire_Close_02}}
 {id=WavePlayer_3, properties {soundWaveAssetPtr=&/ShotgunDrum_Fire_Close_01}}
 {id=WavePlayer_4, properties {soundWaveAssetPtr=&ShotgunDrum_Fire_Close_02}}
 {id=WavePlayer_7, properties {soundWaveAssetPtr=&/ShotgunDrum_Fire_Close_05}}
 {id=WavePlayer_9, properties {soundWaveAssetPtr=&/ShotgunDrum_Fire_Close_03}}
]
properties {
 attenuationSettings=&Fort_Weapon_Projectile_FirstPerson_Athena_Att
 maxDistance=2097152
 soundClassObject=&Fort_WeaponRanged_Fire_1P_PSM
 volumeMultiplier=0.9
 attenuationOverrides {falloffDistance=5000, attenuationShapeExtents {x=1000}}
 concurrencySet [&Weapon_Gunfire_FirstPerson_Conc]
}
