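# Sound cue for Drum Shotgun gunfire, mixing close, medium and distant layers, with the close layer classed as third-person fire.
# Each layer picks a random wave, modulates it, then applies its own attenuation and sound class.
# A quality switch gives the first level three variations per layer and routes the other three levels to a shared mix with two.
# Local and global gunfire concurrency sets govern simultaneous shots.
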SoundCue /Game/Sounds/Fort_Weapons/Shotgun/Drum/ShotgunDrum_Fire_Cue
root=QualityLevel_1
nodes [
 {id=Attenuation_0, children [Modulator_1], properties {attenuationSettings=&Fort_Weapon_Projectile_Close_Athena_Att}}
 {
  id=Attenuation_1
  children [Modulator_0]
  properties {attenuationSettings=&Fort_Weapon_Projectile_Distant_Mono_Athena_Att}
 }
 {
  id=Attenuation_10
  children [Modulator_15]
  properties {attenuationSettings=&Fort_Weapon_Projectile_Close_Athena_Att}
 }
 {
  id=Attenuation_11
  children [Modulator_14]
  properties {attenuationSettings=&Fort_Weapon_Projectile_Distant_Mono_Athena_Att}
 }
 {
  id=Attenuation_12
  children [Modulator_16]
  properties {attenuationSettings=&Fort_Weapon_Projectile_Medium_Mono_Athena_Att}
 }
 {
  id=Attenuation_8
  children [Modulator_2]
  properties {attenuationSettings=&Fort_Weapon_Projectile_Medium_Mono_Athena_Att}
 }
 {id=Mixer_0, children [SoundClass_0,SoundClass_8,SoundClass_1], properties {inputVolume [1,0.7,0.37]}}
 {id=Mixer_6, children [SoundClass_10, SoundClass_12, SoundClass_11], properties {inputVolume [1,0.7,0.37]}}
 {id=Modulator_0, children [Random_10], properties {pitchMax=1, volumeMax=1.9, volumeMin=1.9}}
 {id=Modulator_1, children [Random_0], properties {volumeMax=1.3, volumeMin=1.25}}
 {id=Modulator_14, children [Random_13], properties {pitchMax=1, pitchMin=0.9, volumeMax=1.9, volumeMin=1.9}}
 {id=Modulator_15, children [Random_1], properties {volumeMax=1.3, volumeMin=1.25}}
 {id=Modulator_16, children [Random_12], properties {volumeMax=1.9, volumeMin=1.9}}
 {id=Modulator_2, children [Random_9], properties {volumeMax=1.9, volumeMin=1.9}}
 {id=QualityLevel_1, children [Mixer_0,Mixer_6,Mixer_6,Mixer_6]}
 {id=Random_0, children [WavePlayer_0,WavePlayer_1,WavePlayer_2], properties {weights [1,1,1]}}
 {id=Random_1, children [WavePlayer_3,WavePlayer_4], properties {weights [1,1]}}
 {id=Random_10, children [WavePlayer_26, WavePlayer_27, WavePlayer_28], properties {weights [1,1,1]}}
 {id=Random_12, children [WavePlayer_32, WavePlayer_33], properties {weights [1,1]}}
 {id=Random_13, children [WavePlayer_35, WavePlayer_36], properties {weights [1,1]}}
 {id=Random_9, children [WavePlayer_23, WavePlayer_24, WavePlayer_25], properties {weights [1,1,1]}}
 {
  id=SoundClass_0
  children [Attenuation_0]
  properties {soundClassOverride=&Fort_WeaponRanged_Fire_3P_PSM}
 }
 {
  id=SoundClass_1
  children [Attenuation_1]
  properties {soundClassOverride=&Fort_WeaponRanged_Fire_Distant}
 }
 {
  id=SoundClass_10
  children [Attenuation_10]
  properties {soundClassOverride=&Fort_WeaponRanged_Fire_3P_PSM}
 }
 {
  id=SoundClass_11
  children [Attenuation_11]
  properties {soundClassOverride=&Fort_WeaponRanged_Fire_Distant}
 }
 {
  id=SoundClass_12
  children [Attenuation_12]
  properties {soundClassOverride=&Fort_WeaponRanged_Fire_Distant}
 }
 {
  id=SoundClass_8
  children [Attenuation_8]
  properties {soundClassOverride=&Fort_WeaponRanged_Fire_Distant}
 }
 {id=WavePlayer_0, properties {soundWaveAssetPtr=&/ShotgunDrum_Fire_CloseB_01}}
 {id=WavePlayer_1, properties {soundWaveAssetPtr=&/ShotgunDrum_Fire_CloseB_03}}
 {id=WavePlayer_2, properties {soundWaveAssetPtr=&/ShotgunDrum_Fire_CloseB_02}}
 {id=WavePlayer_23, properties {soundWaveAssetPtr=&/ShotgunDrum_Fire_Medium_01}}
 {id=WavePlayer_24, properties {soundWaveAssetPtr=&/ShotgunDrum_Fire_Medium_02}}
 {id=WavePlayer_25, properties {soundWaveAssetPtr=&/ShotgunDrum_Fire_Medium_03}}
 {id=WavePlayer_26, properties {soundWaveAssetPtr=&/ShotgunDrum_Fire_Distant_01}}
 {id=WavePlayer_27, properties {soundWaveAssetPtr=&/ShotgunDrum_Fire_Distant_02}}
 {id=WavePlayer_28, properties {soundWaveAssetPtr=&/ShotgunDrum_Fire_Distant_03}}
 {id=WavePlayer_3, properties {soundWaveAssetPtr=&ShotgunDrum_Fire_CloseB_01}}
 {id=WavePlayer_32, properties {soundWaveAssetPtr=&ShotgunDrum_Fire_Medium_01}}
 {id=WavePlayer_33, properties {soundWaveAssetPtr=&ShotgunDrum_Fire_Medium_02}}
 {id=WavePlayer_35, properties {soundWaveAssetPtr=&ShotgunDrum_Fire_Distant_01}}
 {id=WavePlayer_36, properties {soundWaveAssetPtr=&ShotgunDrum_Fire_Distant_02}}
 {id=WavePlayer_4, properties {soundWaveAssetPtr=&ShotgunDrum_Fire_CloseB_03}}
]
properties {
 maxDistance=27500
 volumeMultiplier=0.65
 attenuationOverrides {falloffDistance=5000, attenuationShapeExtents {x=1000}}
 concurrencyOverrides {maxCount=3, resolutionRule=StopOldest}
 concurrencySet [&Weapon_Gunfire_Local_Conc, &Weapon_Gunfire_Global_Conc]
}
