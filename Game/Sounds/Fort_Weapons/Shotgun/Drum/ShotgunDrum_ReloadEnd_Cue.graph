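# ShotgunDrum_ReloadEnd_Cue is a sound cue that plays the reload-end sound for the drum shotgun.
# A single wave player feeds the root directly, with no randomization or mixing.
# It uses the shared weapon reload attenuation and the weapon foley sound class, with reduced volume.
# A concurrency override limits simultaneous plays per owner, stopping the oldest instance when the limit is exceeded.
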SoundCue /Game/Sounds/Fort_Weapons/Shotgun/Drum/ShotgunDrum_ReloadEnd_Cue
root=WavePlayer_0
nodes [{id=WavePlayer_0, properties {soundWaveAssetPtr=&/ShotgunDrum_Reload_End_01}}]
properties {
 attenuationSettings=&Fort_Weapon_Reload_Att
 bOverrideConcurrency=true
 maxDistance=2250
 soundClassObject=&Fort_WeaponFoley
 volumeMultiplier=0.752
 attenuationOverrides {falloffDistance=1600}
 concurrencyOverrides {bLimitToOwner=true, maxCount=2, resolutionRule=StopOldest}
}
