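# Sound cue for the magazine insert foley during a drum shotgun reload.
# A single wave player feeds a modulator that fixes pitch slightly below normal, with no random variation.
# It routes to the weapon foley sound class with shared reload attenuation, and its concurrency override prevents new instances from starting once the limit is reached.
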SoundCue /Game/Sounds/Fort_Weapons/Shotgun/Drum/ShotgunDrum_ReloadInsert_Cue
root=Modulator_0
nodes [
 {id=Modulator_0, children [WavePlayer_4], properties {pitchMax=0.9, pitchMin=0.9}}
 {id=WavePlayer_4, properties {soundWaveAssetPtr=&/ShotgunDrum_Reload_Insert_01}}
]
properties {
 attenuationSettings=&Fort_Weapon_Reload_Att
 bOverrideConcurrency=true
 maxDistance=2250
 soundClassObject=&Fort_WeaponFoley
 volumeMultiplier=0.501
 attenuationOverrides {falloffDistance=1600}
 concurrencyOverrides {resolutionRule=PreventNew}
}
